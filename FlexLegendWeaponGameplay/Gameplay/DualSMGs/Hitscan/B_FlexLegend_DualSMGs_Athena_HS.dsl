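; Hitscan dual SMG weapon Blueprint for FlexLegend that adds per-shot effects on top of its parent's hand-fired logic.
; After the parent call and a very short delay, it plays muzzle flash, shell ejection and force feedback unless fire FX are stopped.
; A hand-not-LEFT flag drives the FX, intended to cover the other hand.
; Shells also require no scope or first-person camera and a rendering detail mode above zero.
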(event EventBeginPlay
  (|CallParentFunction) ; This node is disabled and will not be called.\nDrag off pins to build functionality.
)

(event EventTick
  (params (DeltaSeconds :type Float%20%28single-precision%29))
  (|CallParentFunction DeltaSeconds) ; This node is disabled and will not be called.\nDrag off pins to build functionality.
)

(event EventHandFired
  (params (Hand :type EDualWeaponHand%20Enum :default "LEFT") (bPersistantFire :type Boolean))
  (bind _returnvalue (Utilities|Enum|NotEqual(Enum) Hand "LEFT"))
  (|CallParentFunction Hand bPersistantFire)
  ;[0 0 0] Play FX on other Hand, should probably be rolled into base bp/class as an option, if it is already i did not see one
  (Utilities|FlowControl|Delay 0.01)
  (if (not (Variables|Default|GetbStoppedWeaponFireFX))
    (CallFunction|MuzzleFlashFX bPersistantFire _returnvalue)
    (if (and (not (Weapon|IsUsingScopeorFirstPersonCamera)) (> (Rendering|GetRenderingDetailMode) 0))
      (CallFunction|ActivateShellsFX (Variables|Effects%28Shells%29|GetUseShellsOnFire?) _returnvalue)
      (else
        _))
    (FortWeapon|PlayForceFeedback false (Variables|Default|GetbForceFeedbackShouldPersist))))

(fn ConstructionScript
  (|CallParentFunction))
